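# Animation Blueprint for the M_MED_ArcticIceTalus player body, layering deform and animation behavior on the parent mesh pose.
# Two Control Rigs, the second driven by scarf visibility, feed AnimDynamics springs on both calf twist bones, weighted differently in the frontend.
# Outside the frontend, it feeds foot velocity and calf collision to material WPO cloth, adding wind during skydive float.
# Cape backbling tags, parachute state and the face-part modifier set scarf visibility.
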AnimBlueprint /BRCosmetics/Characters/Player/Male/Medium/Bodies/M_MED_ArcticIceTalus/Meshes/M_MED_ArcticIceTalus_AnimBP
target_skeleton=&/M_MED_ArcticIceTalus_Skeleton
graphs [
 AnimGraph {nodes [
   n1 = Control Rig(Source=Copy Pose From Mesh() {bUseAttachedParent=true}) {
    alphaBoolBlend {blendOption=Linear}
    controlRigAssetReference {blueprintRigClass=&/M_MED_ArcticIceTalus_Deform_CR.M_MED_ArcticIceTalus_Deform_CR_C}
   }
   n2 = Control Rig!(Source=@n1, Alpha, custom_cape_equipped:bool=Get Scarf Visible() -> (Scarf Visible:bool)) -> (Pose) {
    alphaBoolBlend {blendOption=Linear}
    controlRigAssetReference {blueprintRigClass=&/M_MED_ArcticIceTalus_CR.M_MED_ArcticIceTalus_CR_C}
   }
   n3 = Get pantlegAlphaFrontend() -> (pantlegAlphaFrontend:double)
   n4 = Anim Dynamics - Bone: dfrm_calf_twist_01_l(ComponentPose=Local To Component(LocalPose=@n2), Alpha=@n3, ExternalForce="0,0,0") {
    bLinearSpring=true
    bOverrideLinearDamping=true
    bUseGravityOverride=true
    linearDampingOverride=0.9999
    linearSpringConstant=175
    simulationSpace=BoneRelative
    alphaBoolBlend {blendOption=Linear}
    boundBone {boneName=dfrm_calf_twist_01_l}
    componentAppliedLinearAccClamp=<3000 3000 3000>
    componentLinearAccScale=<0.1 0.1 0.1>
    componentLinearVelScale=<0.9 0.9 0.9>
    gravityOverride {z=-980}
    physicsBodyDefinitions [{
      boundBone {boneName=dfrm_calf_twist_01_l}
      boxExtents=<60 60 60>
      constraintSetup {angularLimitsMax=<0 0 0>, angularLimitsMin=<0 0 0>, angularTarget=<0 0 0>, linearAxesMax=<5 4 5>, linearAxesMin=<-1.5 -4 -5>}
      localJointOffset=<0 0 0>
     }]
    relativeSpaceBone {boneName=foot_l}
   }
   n5 = Anim Dynamics - Bone: dfrm_calf_twist_01_r(ComponentPose=@n4, Alpha=@n3, ExternalForce="0,0,0") {
    bLinearSpring=true
    bOverrideLinearDamping=true
    bUseGravityOverride=true
    linearDampingOverride=0.9999
    linearSpringConstant=175
    simulationSpace=BoneRelative
    alphaBoolBlend {blendOption=Linear}
    boundBone {boneName=dfrm_calf_twist_01_r}
    componentAppliedLinearAccClamp=<3000 3000 3000>
    componentLinearAccScale=<0.1 0.1 0.1>
    componentLinearVelScale=<0.9 0.9 0.9>
    gravityOverride {z=-980}
    physicsBodyDefinitions [{
      boundBone {boneName=dfrm_calf_twist_01_r}
      boxExtents=<60 60 60>
      constraintSetup {angularLimitsMax=<0 0 0>, angularLimitsMin=<0 0 0>, angularTarget=<0 0 0>, linearAxesMax=<1.5 4 5>, linearAxesMin=<-5 -4 -5>}
      localJointOffset=<0 0 0>
     }]
    relativeSpaceBone {boneName=foot_r}
   }
   Output Pose(Result=Component To Local(ComponentPose=@n5)) {layerGroup=None}
  ]}
 EventGraph {nodes [
   n1 = Event BlueprintUpdateAnimation()
   > n2 = Sequence!(execute) -> (then_0, then_1, then_2:Exec)
   > Is Valid(InputObject=Get PlayerPawn() -> (PlayerPawn:Fort Player Pawn Athena Object Reference))
   > n3 = Branch(Condition=Get bIsInFrontEnd())
   Set WPO Globals Function(
    execute=@n3.else
    Material=Get Mat Instance Dynamic()
    Overlay Material=Get Overlay Mat Instance Dynamic()
    Skeletal Mesh=Get SkelMesh()
    "Local Noise Transform  Location"=Get Local Noise Transform Location() -> (Local Noise Transform Location:Vector)
    Local Noise Offset=Get Local Noise Offset()
   )
   > Set WPO Regional Function!(
    execute
    Material:Material Instance Dynamic Object Reference=Get Mat Instance Dynamic()
    Overlay Material:Material Instance Dynamic Object Reference=Get Overlay Mat Instance Dynamic()
    Skeletal Mesh:Skeletal Mesh Component Object Reference=Get SkelMesh()
    Velocity Transform:Name=foot_r
    Velocity Interp Speed:double=Get Velocity Interp Speed()
    Collision Transform:Name=calf_r
    Collision Transform Axis:EAxis Enum
    Invert Collision Transform Axis:bool=false
    Delta Time X:double=@n1.DeltaTimeX
    Previous Velocity Transform Location:"Vector (by ref)"=Get R1 Velocity Transform Location()
    Previous Velocity Transform Velocity:"Vector (by ref)"=Get R1 Velocity Transform Velocity()
    Region:int=1
   ) -> (then)
   > Set WPO Regional Function(
    Material=Get Mat Instance Dynamic()
    Overlay Material=Get Overlay Mat Instance Dynamic()
    Skeletal Mesh=Get SkelMesh()
    Velocity Transform=foot_l
    Velocity Interp Speed=Get Velocity Interp Speed()
    Collision Transform=calf_l
    Invert Collision Transform Axis=true
    Delta Time X=@n1.DeltaTimeX
    Previous Velocity Transform Location=Get R2 Velocity Transform Location()
    Previous Velocity Transform Velocity=Get R2 Velocity Transform Velocity()
    Region=2
   )
   > n4 = Branch(Condition=Get bIsSkydiveFloating())
   > n5 = Set WPO Pho Wind Intensity(WPO Pho Wind Intensity=1)
   SetScalarParameterValue(
    execute=[@n6.then, @n5.then]
    self=[Get Mat Instance Dynamic(), Get Overlay Mat Instance Dynamic()]
    ParameterName=ClothSim_WindStrength
    Value=Get WPO Pho Wind Intensity()
   )
   n6 = Set WPO Pho Wind Intensity(execute=@n4.else, WPO Pho Wind Intensity=0)
   n7 = Branch(execute=@n2.then_1, Condition=AND Boolean(A=Get bIsParachuteOpened(), B=Get custom_cape_equipped() -> (custom_cape_equipped:bool)))
   > Do Once(Reset=@n7.else)
   > n8 = Cast To FortPlayerPawn(Object=TryGetFortPawn())
   > Is Valid(InputObject=@n8)
   > n9 = Cast To B_Athena_PartModifier_ArcticIce_FaceAcc!(
    execute
    Object=GetModifierForPartType!(self:Fort Player Pawn Object Reference=@n8, PartType:EFortCustomPartType Enum=Face) -> (ReturnValue:Custom Character Part Modifier Object Reference)
   ) -> (then, CastFailed, AsB Athena Part Modifier Arctic Ice Face Acc:B Athena Part Modifier Arctic Ice Face Acc Object Reference) {is=title K2Node_DynamicCast}
   > Set Scarf Visible(Scarf Visible:bool=Get Scarf Visible!(self:B Athena Part Modifier Arctic Ice Face Acc Object Reference=@n9) -> (Scarf Visible)) -> (Output_Get:bool)
   Set pantlegAlphaFrontend(execute=@n2.then_2, pantlegAlphaFrontend:double=SelectFloat(A=0.65, B=1, bPickA=Get bIsInFrontEnd())) -> (Output_Get:double)
   Event BlueprintInitializeAnimation()
   > Set SkelMesh(SkelMesh=GetOwnerSkeletalMeshComponent())
   n10 = TryGetPawnOwner()
   > Is Valid(InputObject=@n10)
   > n11 = Cast To FortPlayerPawnAthena(Object=@n10)
   > Set PlayerPawn(PlayerPawn:Fort Player Pawn Athena Object Reference=@n11) -> (Output_Get:Fort Player Pawn Athena Object Reference)
   Event OnCharacterCustomizationLoaded()
   > n12 = Sequence!(execute) -> (then_0, then_1)
   > n13 = Set SkelMesh(SkelMesh=GetPartSkeletalMeshComponent())
   > n14 = Get Dynamic Material Instance(SkeletalMeshComponent=@n13)
   > Set Mat Instance Dynamic(Mat Instance Dynamic=@n14)
   > n15 = GetOrCreateDynamicOverlayMaterial(MeshComponent=@n13, MaterialName=None)
   > Set Overlay Mat Instance Dynamic(Overlay Mat Instance Dynamic=@n15)
   n16 = Branch(execute=@n12.then_1, Condition=CheckForBackbling().bIsBackblingACape)
   > n17 = Branch(Condition=BackpackItemDefinitionHasAnyMetaTag!(Tags:Array of Gameplay Tag Structures=Make Array("[0]"=(TagName="Cosmetics.MetaTags.ArcticIce.Cape")), bRequiresEquipped:bool=false) -> (ReturnValue:bool))
   > Set custom_cape_equipped(custom_cape_equipped:bool=true) -> (Output_Get:bool)
   > Set Scarf Visible(Scarf Visible=true)
   Set custom_cape_equipped(execute=@n17.else, custom_cape_equipped=false)
   > Set Scarf Visible(Scarf Visible=false)
   Set custom_cape_equipped(execute=@n16.else, custom_cape_equipped=false)
   > Set Scarf Visible(Scarf Visible=false)
   Event OnParachuteClosed()
   > n18 = Branch(Condition=Get custom_cape_equipped())
   > Set Scarf Visible(Scarf Visible=true)
   Set Scarf Visible(execute=@n18.else, Scarf Visible=false)
  ]}
]
variables [
 custom_cape_equipped {type=boolean}
 Local Noise Offset {type=Vector}
 Local Noise Transform Location {type=Vector}
 Mat Instance Dynamic {type=MaterialInstanceDynamic}
 Overlay Mat Instance Dynamic {type=MaterialInstanceDynamic}
 pantlegAlphaFrontend {type=number}
 PlayerPawn {type=FortPlayerPawnAthena}
 R1 Velocity Transform Location {type=Vector}
 R1 Velocity Transform Velocity {type=Vector}
 R2 Velocity Transform Location {type=Vector}
 R2 Velocity Transform Velocity {type=Vector}
 Scarf Visible {type=boolean}
 SkelMesh {type=SkeletalMeshComponent}
 Velocity Interp Speed {type=number, default=30}
 WPO Pho Wind Intensity {type=number}
]
