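# Physics asset holding collision shapes for the cape of the M_MED_ArcticIceTalus player body, as its CapeColliders name indicates.
# Overlapping capsules cover the pelvis, spine, neck, clavicles, thighs, calves and feet, with a sphere on the head.
# Limb and clavicle capsules are mirrored left to right, with especially wide radii on thighs and calves.
# Upper and lower arm bodies carry no shapes, and no constraints link the bodies.
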PhysicsAsset /BRCosmetics/Characters/Player/Male/Medium/Bodies/M_MED_ArcticIceTalus/Meshes/M_MED_ArcticIceTalus_CapeColliders
bodies {
 calf_l [{shapeType=Capsule, length=15, radius=25, shapeName=calf_l_capsule, center=<-15.3251 9.0072 0.0257>, rotation=<rot 90 0 0>}]
 calf_r [{shapeType=Capsule, length=15, radius=25, shapeName=calf_l_capsule, center=<15.3251 -9.0072 -0.0257>, rotation=<rot 90 0 180>}]
 clavicle_l [{shapeType=Capsule, length=8.3719, radius=14.8035, center=<10.1041 0.6502 -0.5792>, rotation=<rot 75.9352 -0.0001 175.5622>}]
 clavicle_r [{shapeType=Capsule, length=8.3719, radius=14.8035, center=<-10.1041 -0.6502 0.5792>, rotation=<rot 75.9352 -0.0001 -4.4378>}]
 foot_l [{shapeType=Capsule, length=14, radius=15, shapeName=foot_l_capsule, center=<-4.0998 7.0178 0.6714>, rotation=<rot 0 0 90>}]
 foot_r [{shapeType=Capsule, length=14, radius=15, shapeName=foot_l_capsule, center=<4.0998 -7.0178 -0.6714>, rotation=<rot 0 0 -90>}]
 head [{shapeType=Sphere, radius=17.3125, center=<11.028 -0.174 0>}]
 lowerarm_l []
 lowerarm_r []
 neck_02 [{shapeType=Capsule, length=13.4375, radius=7.3175, center=<-0.2642 -0.3408 0>, rotation=<rot 90 -171.1193 -175.3978>}]
 pelvis [{shapeType=Capsule, length=8.6562, radius=13.5388, center=<-2.6215 -1.3695 0>}]
 spine_01 [{shapeType=Capsule, length=10.7187, radius=12.4138, center=<-2.6215 -0.7156 0>}]
 spine_02 [
  {shapeType=Capsule, length=10.5937, radius=10.945, center=<-2.6215 0.4993 0.0001>}
  {shapeType=Capsule, length=29.875, radius=11.1013, center=<2.5024 0.4992 3.9825>, rotation=<rot 90 0.0049 0.0049>}
  {shapeType=Capsule, length=29.875, radius=11.1013, center=<2.4704 0.4991 -3.9825>, rotation=<rot 90 -0.0396 -0.0396>}
 ]
 spine_03 [{shapeType=Capsule, length=9.1562, radius=10.7575, center=<-1.953 0.1364 0.0001>}]
 spine_04 [{shapeType=Capsule, length=14.6562, radius=13.6325, center=<2.1467 -1.2748 0.0002>}]
 spine_05 [
  {shapeType=Capsule, length=17.8437, radius=17.195, center=<-1.7514 2.4879 0>}
  {shapeType=Capsule, length=12.7812, radius=17.0625, center=<0.7875 3.1143 0>}
 ]
 thigh_l [{shapeType=Capsule, length=15, radius=20, shapeName=thigh_l_capsule, center=<-23.9334 7.314 1.3009>, rotation=<rot 90 0 0>}]
 thigh_r [{shapeType=Capsule, length=15, radius=20, shapeName=thigh_l_capsule, center=<23.9334 -7.314 -1.3009>, rotation=<rot 90 0 180>}]
 upperarm_l []
 upperarm_r []
}
constraints []
