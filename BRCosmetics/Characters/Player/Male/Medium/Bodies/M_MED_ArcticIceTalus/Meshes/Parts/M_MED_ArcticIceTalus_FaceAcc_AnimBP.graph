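# Anim Blueprint for the ArcticIceTalus face accessory mesh, adding secondary motion to its scarf.
# It copies the attached parent's pose and curves, layering a Body_Mesh (owning component) pose copy from both upper arms.
# Four sprung, damped Anim Dynamics nodes then simulate left and right scarf bones under reduced gravity.
# A scarf constraint chain into a Control Rig and a customization-part pose copy sit unconnected to the output.
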AnimBlueprint /BRCosmetics/Characters/Player/Male/Medium/Bodies/M_MED_ArcticIceTalus/Meshes/Parts/M_MED_ArcticIceTalus_FaceAcc_AnimBP
target_skeleton=/Game/Characters/Player/Male/Medium/Bodies/M_MED_ArcticIceTalus/Meshes/Parts/M_MED_ArcticIceTalus_FaceAcc_Skeleton
graphs [
 AnimGraph {nodes [
   n1 = Layered blend per bone() {layerSetup [{branchFilters [{blendDepth=4, boneName=deform_pec_aim_base_l}, {blendDepth=4, boneName=deform_pec_aim_base_r}]}]}
   n2 = "Constraint - dfrm_scarf_back_mstr_l (PPPPP)"!(
    ComponentPose=Local To Component(LocalPose=@n1)
    Alpha=1
    ConstraintWeights_0=0.6721
    ConstraintWeights_1:float=0.2198
    ConstraintWeights_2:float=0.0662
    ConstraintWeights_3:float=0.0287
    ConstraintWeights_4:float=0.0133
   ) -> (Pose) {
    alphaBoolBlend {blendOption=Linear}
    boneToModify {boneName=dfrm_scarf_back_mstr_l}
    constraintSetup [
     {targetBone {boneName=spine_05}}
     {targetBone {boneName=clavicle_l}}
     {targetBone {boneName=neck_02}}
     {targetBone {boneName=neck_01}}
     {targetBone {boneName=spine_04}}
    ]
    constraintWeights [1,1,1,1,1]
   }
   n3 = "Constraint - dfrm_scarf_back_mstr_r (PPPPP)"!(
    ComponentPose=@n2
    Alpha=1
    ConstraintWeights_0=0.6732
    ConstraintWeights_1:float=0.218
    ConstraintWeights_2:float=0.0665
    ConstraintWeights_3:float=0.0288
    ConstraintWeights_4:float=0.0135
   ) -> (Pose) {
    alphaBoolBlend {blendOption=Linear}
    boneToModify {boneName=dfrm_scarf_back_mstr_r}
    constraintSetup [
     {targetBone {boneName=spine_05}}
     {targetBone {boneName=clavicle_r}}
     {targetBone {boneName=neck_02}}
     {targetBone {boneName=neck_01}}
     {targetBone {boneName=spine_04}}
    ]
    constraintWeights [1,1,1,1,1]
   }
   n4 = "Constraint - dfrm_scarf_side_mstr_l (PPPP)"!(
    ComponentPose=@n3
    Alpha=1
    ConstraintWeights_0=0.5734
    ConstraintWeights_1:float=0.3774
    ConstraintWeights_2:float=0.0274
    ConstraintWeights_3:float=0.0218
   ) -> (Pose) {
    alphaBoolBlend {blendOption=Linear}
    boneToModify {boneName=dfrm_scarf_side_mstr_l}
    constraintSetup [{targetBone {boneName=clavicle_l}}, {targetBone {boneName=spine_05}}, {targetBone {boneName=spine_04}}, {targetBone {boneName=spine_03}}]
    constraintWeights [1,1,1,1]
   }
   n5 = "Constraint - dfrm_scarf_side_mstr_r (PPPP)"!(
    ComponentPose=@n4
    Alpha=1
    ConstraintWeights_0=0.5512
    ConstraintWeights_1:float=0.4011
    ConstraintWeights_2:float=0.0279
    ConstraintWeights_3:float=0.0198
   ) -> (Pose) {
    alphaBoolBlend {blendOption=Linear}
    boneToModify {boneName=dfrm_scarf_side_mstr_r}
    constraintSetup [{targetBone {boneName=clavicle_r}}, {targetBone {boneName=spine_05}}, {targetBone {boneName=spine_04}}, {targetBone {boneName=spine_03}}]
    constraintWeights [1,1,1,1]
   }
   n6 = "Constraint - dfrm_scarf_front_mstr_l (PPP)"!(ComponentPose=@n5, Alpha=1, ConstraintWeights_0=0.4023, ConstraintWeights_1:float=0.3809, ConstraintWeights_2:float=0.2168) -> (Pose) {
    alphaBoolBlend {blendOption=Linear}
    boneToModify {boneName=dfrm_scarf_front_mstr_l}
    constraintSetup [{targetBone {boneName=spine_04}}, {targetBone {boneName=spine_05}}, {targetBone {boneName=clavicle_l}}]
    constraintWeights [1,1,1]
   }
   n7 = "Constraint - dfrm_scarf_front_r (PPP)"!(ComponentPose=@n6, Alpha=1, ConstraintWeights_0=0.4023, ConstraintWeights_1:float=0.3809, ConstraintWeights_2:float=0.2168) -> (Pose) {
    alphaBoolBlend {blendOption=Linear}
    boneToModify {boneName=dfrm_scarf_front_r}
    constraintSetup [{targetBone {boneName=spine_04}}, {targetBone {boneName=spine_05}}, {targetBone {boneName=clavicle_r}}]
    constraintWeights [1,1,1]
   }
   Control Rig(Source=Component To Local(ComponentPose=@n7)) {
    alphaBoolBlend {blendOption=Linear}
    controlRigAssetReference {blueprintRigClass=/Game/Characters/Player/Male/Medium/Bodies/M_MED_ArcticIceTalus/Meshes/Parts/M_MED_ArcticIceTalus_FaceAcc_CtrlRig.M_MED_ArcticIceTalus_FaceAcc_CtrlRig_C}
   }
   Copy Pose From Mesh(SourceMeshComponent=Get skeletal_mesh() -> (skeletal_mesh:Skeletal Mesh Component Object Reference)) {bCopyCurves=true}
   Copy Pose From Mesh() {bCopyCurves=true, bUseAttachedParent=true}
   n8 = Copy Pose From Mesh(SourceMeshComponent=Get Body_Mesh())
   n9 = Copy Pose From Mesh() {bCopyCurves=true, bUseAttachedParent=true}
   n10 = Anim Dynamics - Bone: dyn_L_scarf_3(
    ComponentPose=Local To Component(LocalPose=Layered blend per bone(BasePose=@n9, BlendPoses_0=@n8) {layerSetup [{branchFilters [{boneName=upperarm_l}, {boneName=upperarm_r}]}]})
    Alpha=1
    ExternalForce="0,0,0"
   ) {
    angularDampingOverride=1
    angularSpringConstant=25000
    bAngularSpring=true
    bLinearSpring=true
    bOverrideAngularDamping=true
    bOverrideLinearDamping=true
    bUseGravityOverride=true
    gravityScale=0.35
    linearDampingOverride=1
    linearSpringConstant=20
    alphaBoolBlend {blendOption=Linear}
    boundBone {boneName=dyn_L_scarf_3}
    componentLinearAccScale=<1 1 1>
    componentLinearVelScale=<1 1 1>
    gravityOverride {z=-980}
    physicsBodyDefinitions [{
      boundBone {boneName=dyn_L_scarf_3}
      boxExtents=<20 20 20>
      constraintSetup {angularLimitsMax=<20 20 20>, angularLimitsMin=<-20 -20 -20>, angularTarget=<1 0 0>, linearAxesMax=<2 0 2>, linearAxesMin=<-2 -2 -2>}
      localJointOffset=<1 0 0>
     }]
   }
   n11 = Anim Dynamics - Bone: dyn_L_scarf_4(ComponentPose=@n10, Alpha=1, ExternalForce="0,0,0") {
    angularDampingOverride=1
    angularSpringConstant=25000
    bAngularSpring=true
    bLinearSpring=true
    bOverrideAngularDamping=true
    bOverrideLinearDamping=true
    bUseGravityOverride=true
    gravityScale=0.35
    linearDampingOverride=1
    linearSpringConstant=20
    alphaBoolBlend {blendOption=Linear}
    boundBone {boneName=dyn_L_scarf_4}
    componentLinearAccScale=<1 1 1>
    componentLinearVelScale=<1 1 1>
    gravityOverride {z=-980}
    physicsBodyDefinitions [{
      boundBone {boneName=dyn_L_scarf_4}
      boxExtents=<20 20 20>
      constraintSetup {angularLimitsMax=<20 20 20>, angularLimitsMin=<-20 -20 -20>, angularTarget=<1 0 0>, linearAxesMax=<2 0 2>, linearAxesMin=<-2 -2 -2>}
      localJointOffset=<1 0 0>
     }]
   }
   n12 = Anim Dynamics - Bone: dyn_R_scarf_3(ComponentPose=@n11, Alpha=1, ExternalForce="0,0,0") {
    angularDampingOverride=1
    angularSpringConstant=25000
    bAngularSpring=true
    bLinearSpring=true
    bOverrideAngularDamping=true
    bOverrideLinearDamping=true
    bUseGravityOverride=true
    gravityScale=0.35
    linearDampingOverride=1
    linearSpringConstant=20
    alphaBoolBlend {blendOption=Linear}
    boundBone {boneName=dyn_R_scarf_3}
    componentLinearAccScale=<1 1 1>
    componentLinearVelScale=<1 1 1>
    gravityOverride {z=-980}
    physicsBodyDefinitions [{
      boundBone {boneName=dyn_R_scarf_3}
      boxExtents=<20 20 20>
      constraintSetup {angularLimitsMax=<20 20 20>, angularLimitsMin=<-20 -20 -20>, angularTarget=<1 0 0>, linearAxesMax=<2 2 2>, linearAxesMin=<-2 0 -2>}
      localJointOffset=<1 0 0>
     }]
   }
   n13 = Anim Dynamics - Bone: dyn_R_scarf_4(ComponentPose=@n12, Alpha=1, ExternalForce="0,0,0") {
    angularDampingOverride=1
    angularSpringConstant=25000
    bAngularSpring=true
    bLinearSpring=true
    bOverrideAngularDamping=true
    bOverrideLinearDamping=true
    bUseGravityOverride=true
    gravityScale=0.35
    linearDampingOverride=1
    linearSpringConstant=20
    alphaBoolBlend {blendOption=Linear}
    boundBone {boneName=dyn_R_scarf_4}
    componentLinearAccScale=<1 1 1>
    componentLinearVelScale=<1 1 1>
    gravityOverride {z=-980}
    physicsBodyDefinitions [{
      boundBone {boneName=dyn_R_scarf_4}
      boxExtents=<20 20 20>
      constraintSetup {angularLimitsMax=<20 20 20>, angularLimitsMin=<-20 -20 -20>, angularTarget=<1 0 0>, linearAxesMax=<2 2 2>, linearAxesMin=<-2 0 -2>}
      localJointOffset=<1 0 0>
     }]
   }
   Output Pose(Result=Component To Local(ComponentPose=@n13)) {layerGroup=None}
  ]}
 EventGraph {nodes [
   Event OnCharacterCustomizationLoaded()
   > n1 = Cast To PlayerPawn_Athena(Object=TryGetPawnOwner())
   > Set skeletal_mesh(skeletal_mesh:Skeletal Mesh Component Object Reference=GetSkeletalMeshForPartType(self=@n1)) -> (Output_Get:Skeletal Mesh Component Object Reference)
   Event BlueprintInitializeAnimation()
   > Set Body_Mesh(Body_Mesh=GetOwnerSkeletalMeshComponent())
   Event BlueprintUpdateAnimation()
  ]}
]
variables [
 Body_Mesh {type=SkeletalMeshComponent}
 skeletal_mesh {type=SkeletalMeshComponent}
]
